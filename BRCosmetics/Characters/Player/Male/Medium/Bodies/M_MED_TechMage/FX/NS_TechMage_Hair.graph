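# Niagara cosmetic effect for the male TechMage character that draws glowing green-cyan sprites on the eyes and hair.
# Eyes alternates particles between two fixed offsets using partitioning and fixed facing, while HairDecal scatters translucent sprites over an offset half-cylinder surface.
# Both emitters use camera offsets and curve-driven fades that start overbright.
# User.bisHeadHidden zeroes spawn rate and sprite scale in both, hiding the effect along with the head.
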NiagaraSystem /BRCosmetics/Characters/Player/Male/Medium/Bodies/M_MED_TechMage/FX/NS_TechMage_Hair
dependencies {
 usedDataInterfaces [&NiagaraDataInterfaceCurve, &NiagaraDataInterfaceArrayFloat3]
 usedDynamicInputs []
 usedModules [&SystemState]
}
emitters {
 Eyes {
  emitterUpdateScript [
   &EmitterState()
   &SpawnRate(
    SpawnRate=&MakeCustomFloatFromBool(Bool=$User.bisHeadHidden, True Float=0, False Float=2)
    Use Spawn Probability=false
   )
  ]
  particleSpawnScript [
   &PartitionParticles()
   &InitializeParticle(
    Color=<rgba 0.02 1 0.5 1>
    Lifetime Min=1.5
    Lifetime Mode=::NewEnumerator1
    Position Offset=&SelectVectorFromArray(
     Direct Array Index=$Particles.PartitionParticles.Partition
     Vector Selection Array=DataInterface() {propertyValues {FloatData [<4.3 8.3 2.7>, <4.3 8.3 -2.7>]}}
    )
    Sprite Rotation Angle=0.25
    Sprite Rotation Mode=::NewEnumerator1
    Sprite Size Mode=::NewEnumerator3
    Uniform Sprite Size=3.5
    UsePositionOffset=true
    Write Material Random=true
   )
   &CameraOffset(Camera Offset Amount=3)
  ]
  particleUpdateScript [
   &ParticleState()
   &ScaleColor(
    Scale Alpha=&FloatFromCurve(FloatCurve=DataInterface() {propertyValues {Curve {
        defaultValue=1
        keys [
         <at 0 0>
         <at 0.15 1>
         {arriveTangentWeight=0.1863, interpMode=RCIM_Cubic, leaveTangentWeight=0.1863, tangentMode=RCTM_User, time=0.75, value=1}
         <at 1 0>
        ]
       }}})
    Scale RGB=&VectorFromFloat(Value=&FloatFromCurve(FloatCurve=DataInterface() {propertyValues {Curve {keys [<at 0 3>,<at 1 1>]}}}))
    ScaleA=true
    ScaleRGB=true
   )
   &SpriteFacingAndAlignment(Sprite Facing=<0 1 0>, WriteAlignment=false, WriteFacing=true)
   &ScaleSpriteSize(
    Scale Sprite Size Mode=::NewEnumerator0
    Uniform Scale Factor=&MakeCustomFloatFromBool(Bool=$User.bisHeadHidden, True Float=0, False Float=1)
   )
  ]
  rendererClasses [&NiagaraSpriteRendererProperties]
  renderers [{rendererClass=&NiagaraSpriteRendererProperties}]
 }
 HairDecal {
  emitterUpdateScript [
   &EmitterState()
   &SpawnRate(
    SpawnRate=&MakeCustomFloatFromBool(Bool=$User.bisHeadHidden, True Float=0, False Float=5)
    Use Spawn Probability=false
   )
  ]
  particleSpawnScript [
   &InitializeParticle(
    Color=<rgba 0.02 1 0.5 0.3>
    Lifetime Min=1.5
    Lifetime Mode=::NewEnumerator1
    Sprite Rotation Angle=&RandomRangeFloat(Minimum=-0.05, Maximum=0.05)
    Sprite Rotation Mode=::NewEnumerator1
    Sprite Size Mode=::NewEnumerator1
    UsePositionOffset=false
    Write Material Random=true
   )
   &CameraOffset(Camera Offset Amount=13)
   &ShapeLocation(
    Cylinder Height=16
    Cylinder Radius=3
    Hemicircle Internal Cap=false
    Hemicircle X=true
    MaskSpawnGroup=false
    Offset=<15 0 0>
    Offset Coordinate Space=::Local
    Offset Mode=::NewEnumerator0
    Override Seed=false
    Rotation Axis=<0 1 0>
    Rotation Mode=::NewEnumerator1
    Shape Primitive=::NewEnumerator1
    Surface Only=true
    Use Endcaps In Surface Only Mode=false
    UseOwnerScale=true
   )
  ]
  particleUpdateScript [
   &ParticleState()
   &ScaleColor(
    Scale Alpha=&FloatFromCurve(FloatCurve=DataInterface() {propertyValues {Curve {
        defaultValue=1
        keys [
         <at 0 0>
         <at 0.5 1>
         {arriveTangentWeight=0.1863, interpMode=RCIM_Cubic, leaveTangentWeight=0.1863, tangentMode=RCTM_User, time=0.75, value=1}
         <at 1 0>
        ]
       }}})
    Scale RGB=&VectorFromFloat(Value=&FloatFromCurve(FloatCurve=DataInterface() {propertyValues {Curve {keys [<at 0 3>,<at 1 1>]}}}))
    ScaleA=true
    ScaleRGB=true
   )
   &ScaleSpriteSize(
    Scale Sprite Size Mode=::NewEnumerator0
    Uniform Scale Factor=&MakeCustomFloatFromBool(Bool=$User.bisHeadHidden, True Float=0, False Float=1)
   )
  ]
  rendererClasses [&NiagaraSpriteRendererProperties]
  renderers [{rendererClass=&NiagaraSpriteRendererProperties}]
 }
}
summary {
 systemName=NS_TechMage_Hair
 emitters [
  {emitterName=HairDecal, rendererClasses [&NiagaraSpriteRendererProperties]}
  {emitterName=Eyes, rendererClasses [&NiagaraSpriteRendererProperties]}
 ]
 userVariables [User.bisHeadHidden {type=NiagaraBool}]
}
