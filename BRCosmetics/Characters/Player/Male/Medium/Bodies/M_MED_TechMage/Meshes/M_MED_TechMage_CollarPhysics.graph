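# Physics asset for the male medium TechMage body's collar, covering its back, front and flap dyn_ bones.
# Each collar panel hinges on a tiny base body with tight swing and twist limits, while the flap forms a four-segment chain.
# Free-rotation joints link the flap base to its Driven body and tie the front-right and inside panels to the front panel.
# Unconstrained head and left clavicle shapes provide body volumes around the collar.
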PhysicsAsset /BRCosmetics/Characters/Player/Male/Medium/Bodies/M_MED_TechMage/Meshes/M_MED_TechMage_CollarPhysics
bodies {
 clavicle_l [{shapeType=Capsule, length=18.5465, radius=8.4948, center=<7.6793 -0.2426 -0.7481>, rotation=<rot 90 0 0>}]
 dyn_collarBase_back [{shapeType=Box, extentX=1.01, extentY=1.01, extentZ=1.01, rotation=<rot 0 -90 0>}]
 dyn_collarBase_back_l [{shapeType=Box, extentX=1.01, extentY=1.01, extentZ=1.01, rotation=<rot 0 -90 0>}]
 dyn_collarBase_back_r [{shapeType=Box, extentX=1.01, extentY=1.01, extentZ=1.01, rotation=<rot 0 -90 0>}]
 dyn_collarBase_flap [{shapeType=Box, extentX=1.01, extentY=1.01, extentZ=1.01, rotation=<rot 0 -90 0>}]
 dyn_collarBase_flap_Driven [{shapeType=Box, extentX=1.01, extentY=1.01, extentZ=1.01, rotation=<rot 0 -90 0>}]
 dyn_collarBase_front [{shapeType=Box, extentX=1.01, extentY=1.01, extentZ=1.01, rotation=<rot 0 -90 0>}]
 dyn_collarBase_frontInside [{shapeType=Box, extentX=1.01, extentY=1.01, extentZ=1.01, rotation=<rot 0 -90 0>}]
 dyn_collarBase_front_r [{shapeType=Box, extentX=1.01, extentY=1.01, extentZ=1.01, rotation=<rot 0 -90 0>}]
 dyn_collar_back [{shapeType=Box, extentX=3.4808, extentY=13.3478, extentZ=27.6433, center=<3.7942 0.791 0>, rotation=<rot 0 -89.364 0>}]
 dyn_collar_back_l [{shapeType=Box, extentX=2.3998, extentY=11.0467, extentZ=14.9407, center=<2.8187 0.9978 0.0499>, rotation=<rot -2.7197 -79.893 -2.4232>}]
 dyn_collar_back_r [{shapeType=Box, extentX=2.4995, extentY=17.6214, extentZ=11.9748, center=<2.179 0.1046 -0.0756>, rotation=<rot 4.196 89.9996 -89.9997>}]
 dyn_collar_flap_01 [{shapeType=Box, extentX=1.5627, extentY=7.2181, extentZ=13.311, center=<1.5276 0 0.6766>, rotation=<rot 0 -89.9999 8.4106>}]
 dyn_collar_flap_02 [{shapeType=Box, extentX=1.5627, extentY=4.6072, extentZ=12.9932, center=<1.4717 -0.0001 0>, rotation=<rot 0 -90 0>}]
 dyn_collar_flap_03 [{shapeType=Box, extentX=1.5627, extentY=4.8907, extentZ=11.3521, center=<1.6318 -0.0001 0>, rotation=<rot 0 -90 0>}]
 dyn_collar_flap_04 [{shapeType=Box, extentX=1.5627, extentY=3.7713, extentZ=6.4871, center=<2.0256 0 0.6013>, rotation=<rot 0 -90 0>}]
 dyn_collar_front [{shapeType=Box, extentX=2.2347, extentY=13.6607, extentZ=8.898, center=<2.0099 -0.3728 0.0554>, rotation=<rot 0.3165 -89.9996 -1.9279>}]
 dyn_collar_frontInside [{shapeType=Box, extentX=2.2347, extentY=12.5749, extentZ=8.898, center=<3.7551 -0.0223 0.0413>, rotation=<rot -4.6534 -102.112 -22.0133>}]
 dyn_collar_front_r [{shapeType=Box, extentX=2.4188, extentY=12.1559, extentZ=8.898, center=<0.9957 0.2721 -0.0118>, rotation=<rot 0.4802 -92.2828 4.0923>}]
 head [
  {shapeType=Sphere, radius=4.5, center=<-1.7221 3.8146 0>}
  {shapeType=Sphere, radius=3.392, center=<-6.036 3.6658 0>}
  {shapeType=Capsule, length=1.5949, radius=6.2146, center=<1.6798 -0.1402 0.5836>, rotation=<rot 90 -90 -88.7647>}
 ]
}
constraints [
 {bone1Name=dyn_collar_back, bone2Name=dyn_collarBase_back, swing1LimitDegrees=45, swing2LimitDegrees=15, twistLimitDegrees=8}
 {bone1Name=dyn_collar_back_r, bone2Name=dyn_collarBase_back_r, swing1LimitDegrees=35, swing2LimitDegrees=20, twistLimitDegrees=5}
 {bone1Name=dyn_collar_back_l, bone2Name=dyn_collarBase_back_l, swing1LimitDegrees=40, swing2LimitDegrees=40, twistLimitDegrees=8}
 {
  bone1Name=dyn_collarBase_flap
  bone2Name=dyn_collarBase_flap_Driven
  swing1LimitDegrees=45
  swing1Motion=Free
  swing2LimitDegrees=45
  swing2Motion=Free
  twistLimitDegrees=45
  twistMotion=Free
 }
 {bone1Name=dyn_collar_flap_01, bone2Name=dyn_collarBase_flap, swing1LimitDegrees=20, swing2LimitDegrees=15, twistLimitDegrees=5}
 {bone1Name=dyn_collar_flap_02, bone2Name=dyn_collar_flap_01, swing1LimitDegrees=20, swing2LimitDegrees=15, twistLimitDegrees=5}
 {bone1Name=dyn_collar_flap_03, bone2Name=dyn_collar_flap_02, swing1LimitDegrees=20, swing2LimitDegrees=15, twistLimitDegrees=5}
 {bone1Name=dyn_collar_flap_04, bone2Name=dyn_collar_flap_03, swing1LimitDegrees=20, swing2LimitDegrees=15, twistLimitDegrees=5}
 {bone1Name=dyn_collar_frontInside, bone2Name=dyn_collarBase_frontInside, swing1LimitDegrees=20, swing2LimitDegrees=15, twistLimitDegrees=5}
 {bone1Name=dyn_collar_front, bone2Name=dyn_collarBase_front, swing1LimitDegrees=20, swing2LimitDegrees=15, twistLimitDegrees=5}
 {bone1Name=dyn_collar_front_r, bone2Name=dyn_collarBase_front_r, swing1LimitDegrees=20, swing2LimitDegrees=15, twistLimitDegrees=5}
 {
  bone1Name=dyn_collar_front_r
  bone2Name=dyn_collar_front
  swing1LimitDegrees=45
  swing1Motion=Free
  swing2LimitDegrees=45
  swing2Motion=Free
  twistLimitDegrees=45
  twistMotion=Free
 }
 {
  bone1Name=dyn_collar_frontInside
  bone2Name=dyn_collar_front
  swing1LimitDegrees=45
  swing1Motion=Free
  swing2LimitDegrees=45
  swing2Motion=Free
  twistLimitDegrees=45
  twistMotion=Free
 }
]
