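# Physics asset for the TechMage medium male player body, built around secondary motion of its dynamic skirt and collar bones.
# Capsules and a sphere cover legs, feet, head and left clavicle, while boxes cover the dyn_ bones.
# Angle-limited constraint chains run from small base bodies, with front and side skirt chains rooted to the thighs via sub-thigh bones.
# Free-rotation cross-constraints tie neighbouring skirt chains together, linking the panels laterally.
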PhysicsAsset /BRCosmetics/Characters/Player/Male/Medium/Bodies/M_MED_TechMage/Meshes/M_MED_TechMage_Physics
bodies {
 calf_l [{shapeType=Capsule, length=21.0651, radius=7.1625, center=<-20.9015 -1.5664 -0.3544>, rotation=<rot 0 -92.9878 -89.7931>}]
 calf_r [{shapeType=Capsule, length=21.063, radius=7.0468, center=<20.8742 1.5961 0.4617>, rotation=<rot 0 86.417 -89.839>}]
 clavicle_l [{shapeType=Capsule, length=18.5465, radius=8.4948, center=<7.6793 -0.2426 -0.7481>, rotation=<rot 90 0 0>}]
 dyn_collarBase_back [{shapeType=Box, extentX=1.01, extentY=1.01, extentZ=1.01, rotation=<rot 0 -90 0>}]
 dyn_collarBase_back_l [{shapeType=Box, extentX=1.01, extentY=1.01, extentZ=1.01, rotation=<rot 0 -90 0>}]
 dyn_collarBase_back_r [{shapeType=Box, extentX=1.01, extentY=1.01, extentZ=1.01, rotation=<rot 0 -90 0>}]
 dyn_collarBase_flap [{shapeType=Box, extentX=1.01, extentY=1.01, extentZ=1.01, rotation=<rot 0 -90 0>}]
 dyn_collarBase_flap_Driven [{shapeType=Box, extentX=1.01, extentY=1.01, extentZ=1.01, rotation=<rot 0 -90 0>}]
 dyn_collarBase_frontInside [{shapeType=Box, extentX=1.01, extentY=1.01, extentZ=1.01, rotation=<rot 0 -90 0>}]
 dyn_collar_back [{shapeType=Box, extentX=3.4808, extentY=13.3478, extentZ=27.6433, center=<3.7942 0.791 0>, rotation=<rot 0 -89.364 0>}]
 dyn_collar_back_l [{shapeType=Box, extentX=2.3998, extentY=11.0467, extentZ=14.9407, center=<2.8187 0.9978 0.0499>, rotation=<rot -2.7197 -79.893 -2.4232>}]
 dyn_collar_back_r [{shapeType=Box, extentX=2.4995, extentY=12.2295, extentZ=11.9748, center=<4.2777 -0.1585 2.9635>, rotation=<rot 10.3437 89.9998 -89.9999>}]
 dyn_collar_flap_01 [{shapeType=Box, extentX=1.5627, extentY=7.2181, extentZ=13.311, center=<1.5276 0 0.6766>, rotation=<rot 0 -89.9999 8.4106>}]
 dyn_collar_flap_02 [{shapeType=Box, extentX=1.5627, extentY=4.6072, extentZ=12.9932, center=<1.4717 -0.0001 0>, rotation=<rot 0 -90 0>}]
 dyn_collar_flap_03 [{shapeType=Box, extentX=1.5627, extentY=4.8907, extentZ=11.3521, center=<1.6318 -0.0001 0>, rotation=<rot 0 -90 0>}]
 dyn_collar_flap_04 [{shapeType=Box, extentX=1.5627, extentY=3.7713, extentZ=6.4871, center=<2.0256 0 0.6013>, rotation=<rot 0 -90 0>}]
 dyn_collar_frontInside [{shapeType=Box, extentX=2.2347, extentY=7.3585, extentZ=7.0315, center=<5.4386 -0.8806 0.7426>, rotation=<rot 0.3165 -90 -22.4756>}]
 dyn_midFrontSkirt [{shapeType=Box, extentX=2.006, extentY=11.9673, extentZ=8.1187, center=<3.3719 -0.0001 0.2975>, rotation=<rot 0 89.9999 95.0427>}]
 dyn_midFrontSkirt_Base [{shapeType=Box, extentX=1.01, extentY=1.01, extentZ=1.01, rotation=<rot 0 90 90>}]
 dyn_skirtBackSide_01_Base_l [{shapeType=Box, extentX=1.01, extentY=1.01, extentZ=1.01, rotation=<rot 0 90 90>}]
 dyn_skirtBackSide_01_Base_r [{shapeType=Box, extentX=1.01, extentY=1.01, extentZ=1.01, rotation=<rot 0 90 90>}]
 dyn_skirtBackSide_01_l [{shapeType=Box, extentX=3.6192, extentY=8.4555, extentZ=7.6228, center=<-4.3713 -0.0001 0>, rotation=<rot 0 90 90>}]
 dyn_skirtBackSide_01_r [{shapeType=Box, extentX=3.1979, extentY=8.1295, extentZ=7.4163, center=<4.3347 0.0005 0>, rotation=<rot 0 90 90>}]
 dyn_skirtBackSide_02_l [{shapeType=Box, extentX=3.7951, extentY=12.6782, extentZ=7.3057, center=<-4.3645 0.0002 0>, rotation=<rot 0 90 90>}]
 dyn_skirtBackSide_02_r [{shapeType=Box, extentX=3.4816, extentY=9.3656, extentZ=8.3068, center=<4.2704 0.0004 0>, rotation=<rot 0 90 90>}]
 dyn_skirtBackSide_03_l [{shapeType=Box, extentX=3.4864, extentY=13.2742, extentZ=6.0352, center=<-3.5691 -0.0002 -0.1794>, rotation=<rot -6.4835 89.9999 89.9998>}]
 dyn_skirtBackSide_03_r [{shapeType=Box, extentX=3.5446, extentY=10.7434, extentZ=6.3942, center=<3.7096 -0.1608 0>, rotation=<rot 0 90 90>}]
 dyn_skirtBackSide_04_l [{shapeType=Box, extentX=3.4336, extentY=13.2806, extentZ=3.5454, center=<-3.0362 0 0>, rotation=<rot 0 90 90>}]
 dyn_skirtBackSide_04_r [{shapeType=Box, extentX=3.1737, extentY=11.0005, extentZ=2.2164, center=<2.5912 -0.3057 0>, rotation=<rot 0 90 90>}]
 dyn_skirtBackSide_05_l [{shapeType=Box, extentX=2.9399, extentY=13.6396, extentZ=5.1621, center=<-3.9 0.0001 0>, rotation=<rot 0 90 90>}]
 dyn_skirtBack_01_Base_l [{shapeType=Box, extentX=1.01, extentY=1.01, extentZ=1.01, rotation=<rot 0 90 90>}]
 dyn_skirtBack_01_Base_r [{shapeType=Box, extentX=1.01, extentY=1.01, extentZ=1.01, rotation=<rot 0 90 90>}]
 dyn_skirtBack_01_l [{shapeType=Box, extentX=3.4126, extentY=12.031, extentZ=6.295, center=<-4.0026 -0.0001 0>, rotation=<rot 0 90 90>}]
 dyn_skirtBack_01_r [{shapeType=Box, extentX=3.8096, extentY=12.7699, extentZ=8.4848, center=<4.3284 0.0005 0.0196>, rotation=<rot 0 90 90>}]
 dyn_skirtBack_02_l [{shapeType=Box, extentX=3.4712, extentY=13.8962, extentZ=6.4157, center=<-4.4171 -0.0001 0>, rotation=<rot 0 90 90>}]
 dyn_skirtBack_02_r [{shapeType=Box, extentX=3.6669, extentY=13.185, extentZ=7.6341, center=<4.2832 0.0002 0>, rotation=<rot 0 90 90>}]
 dyn_skirtBack_03_l [{shapeType=Box, extentX=3.2707, extentY=14.1624, extentZ=8.2164, center=<-4.8517 0.0001 0>, rotation=<rot 0 90 90>}]
 dyn_skirtBack_03_r [{shapeType=Box, extentX=4.1049, extentY=13.6828, extentZ=4.6812, center=<3.1351 0.0001 0>, rotation=<rot 0 90 90>}]
 dyn_skirtBack_04_l [{shapeType=Box, extentX=2.9062, extentY=13.2357, extentZ=4.9799, center=<-3.5858 0.0002 0>, rotation=<rot 0 90 90>}]
 dyn_skirtBack_04_r [{shapeType=Box, extentX=2.01, extentY=13.5082, extentZ=3.6938, center=<2.9168 0.0001 -0.4014>, rotation=<rot 0 89.9998 82.1635>}]
 dyn_skirtFront_01_Base_l [{shapeType=Box, extentX=1.01, extentY=1.01, extentZ=1.01, rotation=<rot 0 90 90>}]
 dyn_skirtFront_01_Base_r [{shapeType=Box, extentX=1.01, extentY=1.01, extentZ=1.01, rotation=<rot 0 90 90>}]
 dyn_skirtFront_01_l [{shapeType=Box, extentX=4.3896, extentY=7.6972, extentZ=8.4416, center=<-5.0095 0.2341 0>, rotation=<rot 0 90 90>}]
 dyn_skirtFront_01_r [{shapeType=Box, extentX=4.1961, extentY=8.4476, extentZ=9.5469, center=<5.2759 0.0006 0>, rotation=<rot -4.7253 89.9999 89.9999>}]
 dyn_skirtFront_02_l [{shapeType=Box, extentX=4.3311, extentY=7.2841, extentZ=7.586, center=<-5.4454 -0.2331 0>, rotation=<rot 0 90 90>}]
 dyn_skirtFront_02_r [{shapeType=Box, extentX=2.9113, extentY=10.2604, extentZ=6.7566, center=<5.1449 0.0004 0>, rotation=<rot 0 90 90>}]
 dyn_skirtFront_03_l [{shapeType=Box, extentX=4.459, extentY=9.0023, extentZ=7.4901, center=<-4.9111 -0.0001 0>, rotation=<rot 0 90 90>}]
 dyn_skirtFront_03_r [{shapeType=Box, extentX=3.9943, extentY=6.4019, extentZ=9.4586, center=<4.8421 0.3812 0.0001>, rotation=<rot 0 90 0>}]
 dyn_skirtFront_04_l [{shapeType=Box, extentX=4.2297, extentY=10.2831, extentZ=4.7593, center=<-3.5387 -0.0001 0>, rotation=<rot 0 90 90>}]
 dyn_skirtFront_04_r [{shapeType=Box, extentX=3.251, extentY=4.8422, extentZ=8.3898, center=<3.876 0.0002 0>, rotation=<rot 0 90 0>}]
 dyn_skirtFront_05_l [{shapeType=Box, extentX=5.0228, extentY=10.7459, extentZ=7.4601, center=<-4.0727 -1.09 0>, rotation=<rot 0 93.5271 89.9999>}]
 dyn_skirtSide_01_Base_l [{shapeType=Box, extentX=1.01, extentY=1.01, extentZ=1.01, rotation=<rot 0 90 90>}]
 dyn_skirtSide_01_Base_r [{shapeType=Box, extentX=1.01, extentY=1.01, extentZ=1.01, rotation=<rot 0 90 90>}]
 dyn_skirtSide_01_l [{shapeType=Box, extentX=3.9167, extentY=9.4142, extentZ=6.7412, center=<-4.6289 -0.0002 0>, rotation=<rot 0 90 90>}]
 dyn_skirtSide_01_r [{shapeType=Box, extentX=3.9704, extentY=6.7437, extentZ=7.1628, center=<4.3765 -0.0001 0>, rotation=<rot 8.6589 89.9999 89.9999>}]
 dyn_skirtSide_02_l [{shapeType=Box, extentX=4.0707, extentY=9.6788, extentZ=7.2753, center=<-4.8805 -0.1363 0>, rotation=<rot 0 92.6898 89.9998>}]
 dyn_skirtSide_02_r [{shapeType=Box, extentX=3.8015, extentY=10.5939, extentZ=5.6912, center=<5.2718 -0.0006 0>, rotation=<rot 8.1492 89.9999 89.9999>}]
 dyn_skirtSide_03_l [{shapeType=Box, extentX=2.9521, extentY=10.3056, extentZ=7.3631, center=<-4.9072 -0.0757 0>, rotation=<rot 0 90 90>}]
 dyn_skirtSide_03_r [{shapeType=Box, extentX=3.2666, extentY=10.9954, extentZ=6.7593, center=<4.7098 -0.2728 -0.0455>, rotation=<rot 9.4714 89.9999 89.9999>}]
 dyn_skirtSide_04_l [{shapeType=Box, extentX=4.0276, extentY=10.5775, extentZ=5.8992, center=<-4.2491 -0.7857 0>, rotation=<rot 0 94.28 89.9999>}]
 dyn_skirtSide_04_r [{shapeType=Box, extentX=3.2266, extentY=12.5142, extentZ=5.3129, center=<1.5335 -0.7876 0>, rotation=<rot 6.2156 89.9997 89.9998>}]
 dyn_skirtSide_05_l [{shapeType=Box, extentX=4.7336, extentY=11.2043, extentZ=4.7342, center=<-3.6666 -0.7524 0>, rotation=<rot 0 90 90>}]
 dyn_sub_thigh_l [{shapeType=Box, extentX=1.01, extentY=1.01, extentZ=1.01, rotation=<rot 0 90 90>}]
 dyn_sub_thigh_r [{shapeType=Box, extentX=1.01, extentY=1.01, extentZ=1.01, rotation=<rot 0 90 90>}]
 foot_l [{shapeType=Capsule, length=13.2507, radius=8.814, center=<-4.1 8.3584 0.4378>, rotation=<rot 0 -165.501 -88.9384>}]
 foot_r [{shapeType=Capsule, length=13.2541, radius=8.8207, center=<4.1126 -8.3534 -0.4628>, rotation=<rot 0 -165.4028 -89.1089>}]
 head [
  {shapeType=Sphere, radius=4.5, center=<-1.7221 3.8146 0>}
  {shapeType=Capsule, length=6.3351, radius=6.5601, center=<0.7665 -0.3224 0.5836>, rotation=<rot 90 0 -7.8975>}
 ]
 thigh_l [{shapeType=Capsule, length=28.4134, radius=6.1328, center=<-27.4988 -1.985 -0.3985>, rotation=<rot 90 0 0>}]
 thigh_r [{shapeType=Capsule, length=25.9818, radius=7.416, center=<25.5859 2.4517 0>, rotation=<rot 90 0 0>}]
 thigh_twist_01_l [{shapeType=Capsule, length=21.547, radius=7.0563, center=<0.6587 0 -0.0001>, rotation=<rot 0 -90 -90>}]
 thigh_twist_01_r [{shapeType=Capsule, length=11.0989, radius=7.7242, center=<3.2649 0.1662 -0.8706>, rotation=<rot 0 -85.4899 -89.9999>}]
}
constraints [
 {bone1Name=calf_l, bone2Name=thigh_l, swing1LimitDegrees=45, swing2LimitDegrees=45, twistLimitDegrees=45}
 {bone1Name=foot_l, bone2Name=calf_l, swing1LimitDegrees=45, swing2LimitDegrees=45, twistLimitDegrees=45}
 {bone1Name=calf_r, bone2Name=thigh_r, swing1LimitDegrees=45, swing2LimitDegrees=45, twistLimitDegrees=45}
 {bone1Name=foot_r, bone2Name=calf_r, swing1LimitDegrees=45, swing2LimitDegrees=45, twistLimitDegrees=45}
 {bone1Name=dyn_sub_thigh_l, bone2Name=thigh_l, swing1LimitDegrees=25, swing2LimitDegrees=25, twistLimitDegrees=5}
 {bone1Name=dyn_skirtSide_01_Base_l, bone2Name=dyn_sub_thigh_l, swing1LimitDegrees=25, swing2LimitDegrees=25, twistLimitDegrees=5}
 {bone1Name=dyn_skirtSide_01_l, bone2Name=dyn_skirtSide_01_Base_l, swing1LimitDegrees=25, swing2LimitDegrees=25, twistLimitDegrees=10}
 {bone1Name=dyn_skirtSide_02_l, bone2Name=dyn_skirtSide_01_l, swing1LimitDegrees=25, swing2LimitDegrees=25, twistLimitDegrees=10}
 {bone1Name=dyn_skirtSide_03_l, bone2Name=dyn_skirtSide_02_l, swing1LimitDegrees=25, swing2LimitDegrees=25, twistLimitDegrees=10}
 {bone1Name=dyn_skirtSide_04_l, bone2Name=dyn_skirtSide_03_l, swing1LimitDegrees=25, swing2LimitDegrees=25, twistLimitDegrees=10}
 {bone1Name=dyn_skirtSide_05_l, bone2Name=dyn_skirtSide_04_l, swing1LimitDegrees=25, swing2LimitDegrees=25, twistLimitDegrees=10}
 {bone1Name=dyn_skirtFront_01_Base_l, bone2Name=dyn_sub_thigh_l, swing1LimitDegrees=25, swing2LimitDegrees=25, twistLimitDegrees=10}
 {bone1Name=dyn_skirtFront_01_l, bone2Name=dyn_skirtFront_01_Base_l, swing1LimitDegrees=25, swing2LimitDegrees=25, twistLimitDegrees=10}
 {bone1Name=dyn_skirtFront_02_l, bone2Name=dyn_skirtFront_01_l, swing1LimitDegrees=25, swing2LimitDegrees=25, twistLimitDegrees=10}
 {bone1Name=dyn_skirtFront_03_l, bone2Name=dyn_skirtFront_02_l, swing1LimitDegrees=25, swing2LimitDegrees=25, twistLimitDegrees=10}
 {bone1Name=dyn_skirtFront_04_l, bone2Name=dyn_skirtFront_03_l, swing1LimitDegrees=25, swing2LimitDegrees=25, twistLimitDegrees=10}
 {bone1Name=dyn_skirtFront_05_l, bone2Name=dyn_skirtFront_04_l, swing1LimitDegrees=25, swing2LimitDegrees=25, twistLimitDegrees=10}
 {bone1Name=dyn_sub_thigh_r, bone2Name=thigh_r, swing1LimitDegrees=25, swing2LimitDegrees=25, twistLimitDegrees=10}
 {bone1Name=dyn_skirtSide_01_Base_r, bone2Name=dyn_sub_thigh_r, swing1LimitDegrees=25, swing2LimitDegrees=25, twistLimitDegrees=10}
 {bone1Name=dyn_skirtSide_01_r, bone2Name=dyn_skirtSide_01_Base_r, swing1LimitDegrees=25, swing2LimitDegrees=25, twistLimitDegrees=10}
 {bone1Name=dyn_skirtSide_02_r, bone2Name=dyn_skirtSide_01_r, swing1LimitDegrees=25, swing2LimitDegrees=25, twistLimitDegrees=10}
 {bone1Name=dyn_skirtSide_03_r, bone2Name=dyn_skirtSide_02_r, swing1LimitDegrees=25, swing2LimitDegrees=25, twistLimitDegrees=10}
 {bone1Name=dyn_skirtSide_04_r, bone2Name=dyn_skirtSide_03_r, swing1LimitDegrees=25, swing2LimitDegrees=25, twistLimitDegrees=10}
 {bone1Name=dyn_skirtFront_01_Base_r, bone2Name=dyn_sub_thigh_r, swing1LimitDegrees=25, swing2LimitDegrees=25, twistLimitDegrees=10}
 {bone1Name=dyn_skirtFront_01_r, bone2Name=dyn_skirtFront_01_Base_r, swing1LimitDegrees=25, swing2LimitDegrees=25, twistLimitDegrees=10}
 {bone1Name=dyn_skirtFront_02_r, bone2Name=dyn_skirtFront_01_r, swing1LimitDegrees=25, swing2LimitDegrees=25, twistLimitDegrees=10}
 {bone1Name=dyn_skirtFront_03_r, bone2Name=dyn_skirtFront_02_r, swing1LimitDegrees=25, swing2LimitDegrees=25, twistLimitDegrees=10}
 {bone1Name=dyn_skirtFront_04_r, bone2Name=dyn_skirtFront_03_r, swing1LimitDegrees=25, swing2LimitDegrees=25, twistLimitDegrees=10}
 {bone1Name=dyn_skirtBack_01_r, bone2Name=dyn_skirtBack_01_Base_r, swing1LimitDegrees=25, swing2LimitDegrees=25, twistLimitDegrees=10}
 {bone1Name=dyn_skirtBack_02_r, bone2Name=dyn_skirtBack_01_r, swing1LimitDegrees=25, swing2LimitDegrees=25, twistLimitDegrees=10}
 {bone1Name=dyn_skirtBack_03_r, bone2Name=dyn_skirtBack_02_r, swing1LimitDegrees=25, swing2LimitDegrees=25, twistLimitDegrees=10}
 {bone1Name=dyn_skirtBack_04_r, bone2Name=dyn_skirtBack_03_r, swing1LimitDegrees=25, swing2LimitDegrees=25, twistLimitDegrees=10}
 {bone1Name=dyn_skirtBack_01_l, bone2Name=dyn_skirtBack_01_Base_l, swing1LimitDegrees=25, swing2LimitDegrees=25, twistLimitDegrees=10}
 {bone1Name=dyn_skirtBack_02_l, bone2Name=dyn_skirtBack_01_l, swing1LimitDegrees=25, swing2LimitDegrees=25, twistLimitDegrees=10}
 {bone1Name=dyn_skirtBack_03_l, bone2Name=dyn_skirtBack_02_l, swing1LimitDegrees=25, swing2LimitDegrees=25, twistLimitDegrees=10}
 {bone1Name=dyn_skirtBack_04_l, bone2Name=dyn_skirtBack_03_l, swing1LimitDegrees=25, swing2LimitDegrees=25, twistLimitDegrees=10}
 {bone1Name=dyn_midFrontSkirt, bone2Name=dyn_midFrontSkirt_Base, swing1LimitDegrees=25, swing2LimitDegrees=25, twistLimitDegrees=10}
 {bone1Name=dyn_skirtBackSide_01_r, bone2Name=dyn_skirtBackSide_01_Base_r, swing1LimitDegrees=25, swing2LimitDegrees=25, twistLimitDegrees=10}
 {bone1Name=dyn_skirtBackSide_02_r, bone2Name=dyn_skirtBackSide_01_r, swing1LimitDegrees=25, swing2LimitDegrees=25, twistLimitDegrees=10}
 {bone1Name=dyn_skirtBackSide_03_r, bone2Name=dyn_skirtBackSide_02_r, swing1LimitDegrees=25, swing2LimitDegrees=25, twistLimitDegrees=10}
 {bone1Name=dyn_skirtBackSide_04_r, bone2Name=dyn_skirtBackSide_03_r, swing1LimitDegrees=25, swing2LimitDegrees=25, twistLimitDegrees=10}
 {bone1Name=dyn_skirtBackSide_01_l, bone2Name=dyn_skirtBackSide_01_Base_l, swing1LimitDegrees=25, swing2LimitDegrees=25, twistLimitDegrees=10}
 {bone1Name=dyn_skirtBackSide_02_l, bone2Name=dyn_skirtBackSide_01_l, swing1LimitDegrees=25, swing2LimitDegrees=25, twistLimitDegrees=10}
 {bone1Name=dyn_skirtBackSide_03_l, bone2Name=dyn_skirtBackSide_02_l, swing1LimitDegrees=25, swing2LimitDegrees=25, twistLimitDegrees=10}
 {bone1Name=dyn_skirtBackSide_04_l, bone2Name=dyn_skirtBackSide_03_l, swing1LimitDegrees=25, swing2LimitDegrees=25, twistLimitDegrees=10}
 {bone1Name=dyn_skirtBackSide_05_l, bone2Name=dyn_skirtBackSide_04_l, swing1LimitDegrees=25, swing2LimitDegrees=25, twistLimitDegrees=10}
 {
  bone1Name=dyn_skirtSide_02_r
  bone2Name=dyn_skirtBackSide_02_r
  swing1LimitDegrees=45
  swing1Motion=Free
  swing2LimitDegrees=45
  swing2Motion=Free
  twistLimitDegrees=45
  twistMotion=Free
 }
 {
  bone1Name=dyn_skirtBack_03_r
  bone2Name=dyn_skirtBack_03_l
  swing1LimitDegrees=45
  swing1Motion=Free
  swing2LimitDegrees=45
  swing2Motion=Free
  twistLimitDegrees=45
  twistMotion=Free
 }
 {
  bone1Name=dyn_skirtBack_03_l
  bone2Name=dyn_skirtBackSide_04_l
  swing1LimitDegrees=45
  swing1Motion=Free
  swing2LimitDegrees=45
  swing2Motion=Free
  twistLimitDegrees=45
  twistMotion=Free
 }
 {
  bone1Name=dyn_skirtBackSide_03_l
  bone2Name=dyn_skirtSide_03_l
  swing1LimitDegrees=45
  swing1Motion=Free
  swing2LimitDegrees=45
  swing2Motion=Free
  twistLimitDegrees=45
  twistMotion=Free
 }
 {
  bone1Name=dyn_skirtSide_04_l
  bone2Name=dyn_skirtFront_04_l
  swing1LimitDegrees=45
  swing1Motion=Free
  swing2LimitDegrees=45
  swing2Motion=Free
  twistLimitDegrees=45
  twistMotion=Free
 }
 {
  bone1Name=dyn_skirtSide_03_r
  bone2Name=dyn_skirtBackSide_04_r
  swing1LimitDegrees=45
  swing1Motion=Free
  swing2LimitDegrees=45
  swing2Motion=Free
  twistLimitDegrees=45
  twistMotion=Free
 }
 {
  bone1Name=dyn_skirtSide_04_l
  bone2Name=dyn_skirtBackSide_05_l
  swing1LimitDegrees=45
  swing1Motion=Free
  swing2LimitDegrees=45
  swing2Motion=Free
  twistLimitDegrees=45
  twistMotion=Free
 }
 {
  bone1Name=thigh_twist_01_r
  bone2Name=thigh_r
  swing1LimitDegrees=45
  swing1Motion=Free
  swing2LimitDegrees=45
  swing2Motion=Free
  twistLimitDegrees=45
  twistMotion=Free
 }
 {
  bone1Name=thigh_twist_01_l
  bone2Name=thigh_l
  swing1LimitDegrees=45
  swing1Motion=Free
  swing2LimitDegrees=45
  swing2Motion=Free
  twistLimitDegrees=45
  twistMotion=Free
 }
 {
  bone1Name=dyn_skirtSide_05_l
  bone2Name=dyn_skirtFront_05_l
  swing1LimitDegrees=45
  swing1Motion=Free
  swing2LimitDegrees=45
  swing2Motion=Free
  twistLimitDegrees=45
  twistMotion=Free
 }
 {bone1Name=dyn_collar_back, bone2Name=dyn_collarBase_back, swing1LimitDegrees=45, swing2LimitDegrees=15, twistLimitDegrees=8}
 {bone1Name=dyn_collar_back_r, bone2Name=dyn_collarBase_back_r, swing1LimitDegrees=35, swing2LimitDegrees=20, twistLimitDegrees=5}
 {bone1Name=dyn_collar_back_l, bone2Name=dyn_collarBase_back_l, swing1LimitDegrees=40, swing2LimitDegrees=40, twistLimitDegrees=8}
 {
  bone1Name=dyn_collarBase_flap
  bone2Name=dyn_collarBase_flap_Driven
  swing1LimitDegrees=45
  swing1Motion=Free
  swing2LimitDegrees=45
  swing2Motion=Free
  twistLimitDegrees=45
  twistMotion=Free
 }
 {bone1Name=dyn_collar_flap_01, bone2Name=dyn_collarBase_flap, swing1LimitDegrees=20, swing2LimitDegrees=15, twistLimitDegrees=5}
 {bone1Name=dyn_collar_flap_02, bone2Name=dyn_collar_flap_01, swing1LimitDegrees=20, swing2LimitDegrees=15, twistLimitDegrees=5}
 {bone1Name=dyn_collar_flap_03, bone2Name=dyn_collar_flap_02, swing1LimitDegrees=20, swing2LimitDegrees=15, twistLimitDegrees=5}
 {bone1Name=dyn_collar_flap_04, bone2Name=dyn_collar_flap_03, swing1LimitDegrees=20, swing2LimitDegrees=15, twistLimitDegrees=5}
 {bone1Name=dyn_collar_frontInside, bone2Name=dyn_collarBase_frontInside, swing1LimitDegrees=20, swing2LimitDegrees=15, twistLimitDegrees=5}
 {
  bone1Name=dyn_skirtBack_03_r
  bone2Name=dyn_skirtBackSide_03_r
  swing1LimitDegrees=45
  swing1Motion=Free
  swing2LimitDegrees=45
  swing2Motion=Free
  twistLimitDegrees=45
  twistMotion=Free
 }
 {
  bone1Name=dyn_skirtBackSide_03_r
  bone2Name=dyn_skirtSide_03_r
  swing1LimitDegrees=45
  swing1Motion=Free
  swing2LimitDegrees=45
  swing2Motion=Free
  twistLimitDegrees=45
  twistMotion=Free
 }
]
