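# Post-process style AnimBP for the TechMage cosmetic body that adds secondary motion to its collar and skirt over a pose copied from the attached parent mesh.
# Head rotation and aiming drive additive collar bone offsets, followed by collar RigidBody physics and a cached pose.
# Movement and crouch bools blend between skirt RigidBody setups, with posed left skirt bones for crouching.
# Skydiving and sliding alter gravity or random external force; skydive and parachute events reset dynamics.
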AnimBlueprint /BRCosmetics/Characters/Player/Male/Medium/Bodies/M_MED_TechMage/Meshes/M_MED_TechMage_Skeleton_AnimBP
target_skeleton=&/M_MED_TechMage_Skeleton
graphs [
 AnimGraph {nodes [
   RigidBody!(ComponentPose, Alpha, OverrideWorldGravity:Vector="0,0,-980", ExternalForce) -> (Pose) {
    bClampLinearTranslationLimitToRefPose=true
    bOverrideWorldGravity=true
    overridePhysicsAsset=&/M_MED_TechMage_CrouchPhysics
    simulationSpace=BaseBoneSpace
    alphaBoolBlend {blendOption=Linear}
    baseBoneRef {boneName=pelvis}
    componentAppliedLinearAccClamp=<500 500 2000>
    componentLinearAccScale=<0.3 0.3 0.5>
    componentLinearVelScale=<0.3 0.3 0.35>
    overrideWorldGravity {z=-980}
   }
   n1 = RigidBody(
    ComponentPose=Local To Component(LocalPose=Use cached pose 'preCrouch'())
    Alpha=Get rigidbody_alfa() -> (rigidbody_alfa:double)
    OverrideWorldGravity="0,0,-980"
    ExternalForce=Get skydiving vector force() -> (skydiving vector force:Vector)
   ) {
    bClampLinearTranslationLimitToRefPose=true
    bOverrideWorldGravity=true
    overridePhysicsAsset=&/M_MED_TechMage_Physics
    simulationSpace=BaseBoneSpace
    alphaBoolBlend {blendOption=Linear}
    baseBoneRef {boneName=pelvis}
    componentAppliedLinearAccClamp=<500 500 2000>
    componentLinearAccScale=<0.3 0.3 0.5>
    componentLinearVelScale=<0.3 0.3 0.35>
    overrideWorldGravity {z=-980}
   }
   n2 = Component To Local(ComponentPose=@n1)
   n3 = "Transform (Modify) Bone - Bone: dyn_skirtFront_01_Base_l"(
    ComponentPose=Local To Component(LocalPose=Use cached pose 'preCrouch'())
    Alpha=1
    Translation="-4.811,3.942,3.990"
    Rotation="32,0,0.000"
    Scale="1,1,1"
   ) {
    rotationMode=BMM_Additive
    rotationSpace=BCS_BoneSpace
    translationMode=BMM_Additive
    translationSpace=BCS_BoneSpace
    alphaBoolBlend {blendOption=Linear}
    boneToModify {boneName=dyn_skirtFront_01_Base_l}
    rotation {pitch=32}
    translation=<-4.811 3.942 3.99>
   }
   n4 = "Transform (Modify) Bone - Bone: dyn_skirtFront_02_l"(ComponentPose=@n3, Alpha=1, Translation="-0.593,-1.756,1.934", Rotation="28.374,29.902,1.403", Scale="1,1,1") {
    rotationMode=BMM_Additive
    rotationSpace=BCS_BoneSpace
    translationMode=BMM_Additive
    translationSpace=BCS_BoneSpace
    alphaBoolBlend {blendOption=Linear}
    boneToModify {boneName=dyn_skirtFront_02_l}
    rotation=<rot 28.374 29.9021 1.403>
    translation=<-0.593 -1.756 1.934>
   }
   n5 = "Transform (Modify) Bone - Bone: dyn_skirtSide_01_Base_l"(ComponentPose=@n4, Alpha=1, Translation="0.000,2,-0.000", Rotation="16.481,-7.156,-2.041", Scale="1,1,1") {
    rotationMode=BMM_Additive
    rotationSpace=BCS_BoneSpace
    translationMode=BMM_Additive
    translationSpace=BCS_BoneSpace
    alphaBoolBlend {blendOption=Linear}
    boneToModify {boneName=dyn_skirtSide_01_Base_l}
    rotation=<rot 16.481 -7.1561 -2.041>
    translation {y=2}
   }
   n6 = "Transform (Modify) Bone - Bone: dyn_skirtSide_02_l"(ComponentPose=@n5, Alpha=1, Translation="2.5,1.624,8", Rotation="50,20,0", Scale="1,1,1") {
    rotationMode=BMM_Additive
    rotationSpace=BCS_BoneSpace
    translationMode=BMM_Additive
    translationSpace=BCS_BoneSpace
    alphaBoolBlend {blendOption=Linear}
    boneToModify {boneName=dyn_skirtSide_02_l}
    rotation {pitch=50, yaw=20}
    translation=<2.5 1.624 8>
   }
   n7 = "Transform (Modify) Bone - Bone: dyn_skirtBackSide_01_Base_l"(ComponentPose=@n6, Alpha=1, Translation="0,0,0", Rotation="0,0,0", Scale="1,1,1") {
    rotationMode=BMM_Additive
    rotationSpace=BCS_BoneSpace
    translationMode=BMM_Additive
    translationSpace=BCS_BoneSpace
    alphaBoolBlend {blendOption=Linear}
    boneToModify {boneName=dyn_skirtBackSide_01_Base_l}
   }
   n8 = "Transform (Modify) Bone - Bone: dyn_skirtBackSide_02_l"(ComponentPose=@n7, Alpha=1, Translation="1,0,0", Rotation="0,15,0", Scale="1,1,1") {
    rotationMode=BMM_Additive
    rotationSpace=BCS_BoneSpace
    translationMode=BMM_Additive
    translationSpace=BCS_BoneSpace
    alphaBoolBlend {blendOption=Linear}
    boneToModify {boneName=dyn_skirtBackSide_02_l}
    rotation {yaw=15}
    translation {x=1}
   }
   n9 = RigidBody!(ComponentPose=@n8, Alpha, ExternalForce) -> (Pose) {
    bClampLinearTranslationLimitToRefPose=true
    bOverrideWorldGravity=true
    overridePhysicsAsset=&M_MED_TechMage_CrouchPhysics
    simulationSpace=BaseBoneSpace
    alphaBoolBlend {blendOption=Linear}
    baseBoneRef {boneName=pelvis}
    componentAppliedLinearAccClamp=<5000 5000 5000>
    componentLinearAccScale=<0.3 0.3 0.6>
    componentLinearVelScale=<0.5 0.5 0.4>
    overrideWorldGravity {z=-980}
   }
   n10 = Component To Local(ComponentPose=@n9)
   n11 = Get crouchAlpha() -> (crouchAlpha:bool)
   n12 = Get bIsSlopeSliding() -> (bIsSlopeSliding:bool)
   n13 = Get bIsSkyDiving()
   n14 = Get bIsSkydiveFloating()
   n15 = Get bIsSkydiveDiveMode()
   n16 = RigidBody!(
    ComponentPose=Local To Component(LocalPose=Use cached pose 'preCrouch'())
    Alpha=0.85
    OverrideWorldGravity:Vector=SelectVector(
     A="-20,0,-250"
     B=SelectVector(A="-100,0,-100", B="0,0,-980", bPickA=@n12)
     bPickA=OR Boolean!(A=@n13, B=@n15, C:bool=@n14, D:bool) -> (ReturnValue)
    )
    ExternalForce=Get skydiving vector force()
   ) -> (Pose) {
    bClampLinearTranslationLimitToRefPose=true
    bOverrideWorldGravity=true
    overridePhysicsAsset=&M_MED_TechMage_CrouchPhysics
    simulationSpace=BaseBoneSpace
    alphaBoolBlend {blendOption=Linear}
    baseBoneRef {boneName=pelvis}
    componentAppliedLinearAccClamp=<500 250 2000>
    componentLinearAccScale=<0.3 0.3 0.5>
    componentLinearVelScale=<0.3 0.3 0.35>
    overrideWorldGravity {z=-980}
   }
   Output Pose(Result=Blend Poses by bool(
     BlendPose_0=Component To Local(ComponentPose=@n16)
     BlendPose_1=Blend Poses by bool(BlendPose_0=@n10, BlendPose_1=@n2, BlendTime_0=0.25, BlendTime_1=0.25, bActiveValue=@n11) {blendType=Linear}
     bActiveValue=Get Not_2D_Moving() -> (Not_2D_Moving:bool)
    ) {blendType=Linear}) {layerGroup=None}
   n17 = Get headSwingDeltaBack() -> (headSwingDeltaBack:double)
   n18 = "Constraint - dyn_collarBase_back_l (P)"(ComponentPose=Local To Component(LocalPose=Copy Pose From Mesh() {bUseAttachedParent=true}), Alpha=0.5, ConstraintWeights_0=1) {
    alphaBoolBlend {blendOption=Linear}
    boneToModify {boneName=dyn_collarBase_back_l}
    constraintSetup [{targetBone {boneName=clavicle_l}}]
    constraintWeights [1]
   }
   n19 = "Constraint - dyn_collarBase_back_r (P)"(ComponentPose=@n18, Alpha=0.25, ConstraintWeights_0=1) {
    alphaBoolBlend {blendOption=Linear}
    boneToModify {boneName=dyn_collarBase_back_r}
    constraintSetup [{targetBone {boneName=clavicle_r}}]
    constraintWeights [1]
   }
   n20 = "Constraint - dyn_collarBase_front_r (P)"(ComponentPose=@n19, Alpha=0.15, ConstraintWeights_0=1) {
    alphaBoolBlend {blendOption=Linear}
    boneToModify {boneName=dyn_collarBase_front_r}
    constraintSetup [{targetBone {boneName=clavicle_r}}]
    constraintWeights [1]
   }
   n21 = Get headSwingDeltaFront() -> (headSwingDeltaFront:double)
   n22 = "Transform (Modify) Bone - Bone: dyn_collarBase_front"(ComponentPose=@n20, Alpha=@n21, Translation="-1.5,2,0", Rotation="0,20,0.000", Scale="1,1,1") {
    rotationMode=BMM_Additive
    rotationSpace=BCS_BoneSpace
    translationMode=BMM_Additive
    translationSpace=BCS_BoneSpace
    alphaBoolBlend {blendOption=Linear}
    boneToModify {boneName=dyn_collarBase_front}
    rotation {yaw=20}
    translation=<-1.5 2>
   }
   n23 = "Transform (Modify) Bone - Bone: dyn_collarBase_frontInside"(ComponentPose=@n22, Alpha=@n21, Translation="0,0,2.5", Rotation="0,25,10", Scale="1,1,1") {
    rotationMode=BMM_Additive
    rotationSpace=BCS_BoneSpace
    translationMode=BMM_Additive
    translationSpace=BCS_BoneSpace
    alphaBoolBlend {blendOption=Linear}
    alphaScaleBiasClamp {bMapRange=true, inRange {max=0.75, min=0.125}}
    boneToModify {boneName=dyn_collarBase_frontInside}
    rotation {roll=10, yaw=25}
    translation {z=2.5}
   }
   n24 = "Transform (Modify) Bone - Bone: dyn_collarBase_flap_Driven"(ComponentPose=@n23, Alpha=@n21, Translation="0,1,1", Rotation="0.000,5.000,-3", Scale="1,1,1") {
    rotationMode=BMM_Additive
    rotationSpace=BCS_BoneSpace
    translationMode=BMM_Additive
    translationSpace=BCS_BoneSpace
    alphaBoolBlend {blendOption=Linear}
    alphaScaleBiasClamp {bMapRange=true, inRange {max=0.75, min=0.125}}
    boneToModify {boneName=dyn_collarBase_flap_Driven}
    rotation {roll=-3, yaw=5}
    translation {y=1, z=1}
   }
   n25 = "Transform (Modify) Bone - Bone: dyn_collarBase_front_r"(ComponentPose=@n24, Alpha=Get headTwist_r() -> (headTwist_r:double), Translation="0.5,4,0", Rotation="0,20,0", Scale="1,1,1") {
    rotationMode=BMM_Additive
    rotationSpace=BCS_BoneSpace
    translationMode=BMM_Additive
    translationSpace=BCS_BoneSpace
    alphaBoolBlend {blendOption=Linear}
    alphaScaleBiasClamp {bMapRange=true, inRange {max=0.75, min=0.125}}
    boneToModify {boneName=dyn_collarBase_front_r}
    rotation {yaw=20}
    translation=<0.5 4>
   }
   n26 = "Transform (Modify) Bone - Bone: dyn_collarBase_frontInside"(ComponentPose=@n25, Alpha=Get aimAlpha() -> (aimAlpha:double), Translation="0,2,0", Rotation="0,0,15", Scale="1,1,1") {
    rotationMode=BMM_Additive
    rotationSpace=BCS_BoneSpace
    translationMode=BMM_Additive
    translationSpace=BCS_BoneSpace
    alphaBoolBlend {blendOption=Linear}
    alphaScaleBiasClamp {bInterpResult=true, interpSpeedDecreasing=5, interpSpeedIncreasing=5}
    boneToModify {boneName=dyn_collarBase_frontInside}
    rotation {roll=15}
    translation {y=2}
   }
   n27 = "Transform (Modify) Bone - Bone: dyn_collarBase_back_r"(ComponentPose=@n26, Alpha=@n17, Translation="-5,5,0", Rotation="0,30,0", Scale="1,1,1") {
    rotationMode=BMM_Additive
    rotationSpace=BCS_BoneSpace
    translationMode=BMM_Additive
    translationSpace=BCS_BoneSpace
    alphaBoolBlend {blendOption=Linear}
    alphaScaleBiasClamp {bMapRange=true, outRange {max=0.8}}
    boneToModify {boneName=dyn_collarBase_back_r}
    rotation {yaw=30}
    translation=<-5 5>
   }
   n28 = "Transform (Modify) Bone - Bone: dyn_collarBase_back"(ComponentPose=@n27, Alpha=@n17, Translation="-3,-1.25,0", Rotation="0,18,0", Scale="1,1,1") {
    rotationMode=BMM_Additive
    rotationSpace=BCS_BoneSpace
    translationMode=BMM_Additive
    translationSpace=BCS_BoneSpace
    alphaBoolBlend {blendOption=Linear}
    alphaScaleBiasClamp {bMapRange=true, inRange {max=0.6}}
    boneToModify {boneName=dyn_collarBase_back}
    rotation {yaw=18}
    translation=<-3 -1.25>
   }
   n29 = "Transform (Modify) Bone - Bone: dyn_collarBase_back_l"(ComponentPose=@n28, Alpha=@n17, Translation="-4,3,0", Rotation="-10,20.000,-0.000", Scale="1,1,1") {
    rotationMode=BMM_Additive
    rotationSpace=BCS_BoneSpace
    translationMode=BMM_Additive
    translationSpace=BCS_BoneSpace
    alphaBoolBlend {blendOption=Linear}
    alphaScaleBiasClamp {bMapRange=true, inRange {max=0.75, min=0.125}, outRange {max=0.5}}
    boneToModify {boneName=dyn_collarBase_back_l}
    rotation {pitch=-10, yaw=20}
    translation=<-4 3>
   }
   n30 = Get bIsSlopeSliding()
   n31 = Get bIsSkydiveDiveMode()
   n32 = Get bIsSkyDiving()
   n33 = RigidBody(
    ComponentPose=@n29
    Alpha=0.5
    OverrideWorldGravity=SelectVector(A="-20,0,-250", B=SelectVector(A="-100,0,-100", B="0,0,-980", bPickA=@n30), bPickA=OR Boolean(A=@n32, B=@n31))
   ) {
    bClampLinearTranslationLimitToRefPose=true
    bOverrideWorldGravity=true
    overridePhysicsAsset=&/M_MED_TechMage_CollarPhysics
    simulationSpace=BaseBoneSpace
    alphaBoolBlend {blendOption=Linear}
    baseBoneRef {boneName=root}
    componentAppliedLinearAccClamp=<5000 5000 5000>
    componentLinearAccScale=<0.3 0.3 0.6>
    componentLinearVelScale=<0.5 0.5 0.4>
    overrideWorldGravity {z=-980}
   }
   Save cached pose 'preCrouch'(Pose=Component To Local(ComponentPose=@n33)) {cacheName=preCrouch}
   Break Fort Anim Input Skydiving External Force(FortAnimInput_SkydivingExternalForce=Get SkydivingExternalForceInput()) -> (
    bUseSkydivingVectorForce=~
    SkydivingVectorForce=~
    HeadToPelvisDirection=~
    FloatingMultiplier=~
    FloatingAdditive=~
    DivingMultiplier=~
    DivingAdditive=~
    ParachutingMultiplier=~
    ParachutingAdditive=~
   )
   Get bIsRelaxedLevel1() -> (bIsRelaxedLevel1:bool)
   Get rigidbody_alfa()
   Get headPitchAim() -> (headPitchAim:double)
  ]}
 EventGraph {nodes [
   Event BlueprintUpdateAnimation()
   > n1 = Sequence!(execute) -> (then_0, then_1, then_2:Exec, then_3:Exec, then_4:Exec, then_5:Exec)
   n2 = GetOwnerSkeletalMeshComponent()
   n3 = BreakTransform(InTransform=GetDeltaTransformFromRefPose(self=@n2, BoneName=head, BaseName=spine_05))
   n4 = BreakRotator(InRot=@n3.Rotation)
   > Set headSwingDeltaBack(headSwingDeltaBack:double=MapRangeClamped(Value=@n4.Yaw, InRangeA=0, InRangeB=-45)) -> (Output_Get:double)
   > Set headSwingDeltaFront(headSwingDeltaFront:double=MapRangeClamped(Value=@n4.Yaw, InRangeA=0, InRangeB=25)) -> (Output_Get:double)
   n5 = BreakRotator(InRot=@n3.Rotation)
   > Set headTwist_r(headTwist_r:double=MapRangeClamped(Value=@n5, InRangeA=0, InRangeB=-61)) -> (Output_Get:double)
   > Set headPitchAim(headPitchAim:double=MapRangeClamped(Value=@n5.Pitch, InRangeA=0, InRangeB=37)) -> (Output_Get:double)
   n6 = BreakRotator(InRot=BreakTransform(InTransform=GetDeltaTransformFromRefPose(self=@n2, BoneName=upperarm_r, BaseName=clavicle_r)).Rotation)
   Set armSwingDelta_r(execute=@n1.then_1, armSwingDelta_r:double=MapRangeClamped(Value="Absolute (Float)"(A=@n6.Yaw), InRangeA=45, InRangeB=125)) -> (Output_Get:double)
   n7 = BreakRotator(InRot=BreakTransform(InTransform=GetDeltaTransformFromRefPose(self=@n2, BoneName=upperarm_l, BaseName=clavicle_l)).Rotation)
   Set armSwingDelta_l(execute=@n1.then_2, armSwingDelta_l:double=MapRangeClamped(Value="Absolute (Float)"(A=@n7.Yaw), InRangeA=5, InRangeB=77)) -> (Output_Get:double)
   Set crouchAlpha(execute=@n1.then_3, crouchAlpha:bool=Get bIsCrouching()) -> (Output_Get:bool)
   > Set aimAlpha(aimAlpha:double="To Float (Boolean)"(InBool=Get bIsTargeting())) -> (Output_Get:double)
   n8 = Branch(execute=@n1.then_4, Condition=Get bCanPlayCustomAnimations() -> (bCanPlayCustomAnimations:bool))
   n9 = Branch(execute=[@n8.else, @n8.then], Condition=Get bIsSkyDiving())
   > n10 = Branch(Condition=Get bIsSkydiveFloating())
   n11 = RandomFloatInRangeFromStream!(Stream:"Random Stream Structure (by ref)"=Get random_stream() -> (random_stream:Random Stream Structure), Min:float=-20000, Max:float=20000) -> (ReturnValue:float)
   n12 = MakeVector(X=@n11, Y=@n11)
   > Set skydiving vector force(skydiving vector force:Vector=@n12) -> (Output_Get:Vector)
   n13 = Branch(execute=@n10.else, Condition=Get bIsSkydiveDiveMode())
   n14 = Get Direction Between Sockets!(
    Component:Skeletal Mesh Component Object Reference=GetOwnerSkeletalMeshComponent()
    SocketOrBoneNameFrom:Name=head
    SocketOrBoneNameTo:Name=pelvis
   ) -> (ReturnValue:Vector) {is=title K2Node_CallFunction}
   > Set skydiving vector force(skydiving vector force=vector * vector(A:Vector=vector * vector(A=@n12, B=@n14), B:Vector="0.2,0.2,0.2") -> (ReturnValue:Vector) {is=title K2Node_CommutativeAssociativeBinaryOperator})
   > Set rigidbody_alfa(rigidbody_alfa:double=0.25) -> (Output_Get:double)
   Set skydiving vector force(execute=@n13.else, skydiving vector force=@n12)
   Set skydiving vector force(execute=@n9.else, skydiving vector force=vector * vector(A="0, 0, 0", B="0, 0, 0"))
   > Set rigidbody_alfa(rigidbody_alfa=0.9)
   Set Not_2D_Moving(execute=@n1.then_5, Not_2D_Moving:bool=NOT Boolean(A=Get bIsAccelerating2D())) -> (Output_Get:bool)
   n15 = Event OnBeginSkydiving()
   Do Once(execute=[@n15.then, @n18.then], Reset=@n16.then)
   > Delay(Duration=1.4)
   > ForceClothNextUpdateTeleportAndReset(self:Skeletal Mesh Component Object Reference=Get MeshToCopy_0() -> (MeshToCopy_0:Skeletal Mesh Component Object Reference))
   > ResetDynamics(InTeleportType=TeleportPhysics)
   > n16 = ResetDynamics(InTeleportType=ResetPhysics)
   Event OnEndSkydiving()
   > Do Once(Reset=@n17.then)
   > Delay(Duration=0.25)
   > ForceClothNextUpdateTeleportAndReset(self=Get MeshToCopy_0())
   > ResetDynamics(InTeleportType=TeleportPhysics)
   > n17 = ResetDynamics(InTeleportType=ResetPhysics)
   Event BlueprintInitializeAnimation()
   > Set MeshToCopy_0(MeshToCopy_0:Skeletal Mesh Component Object Reference=GetPartSkeletalMeshComponent()) -> (Output_Get:Skeletal Mesh Component Object Reference)
   n18 = Event OnParachuteOpened()
   TryGetPawnOwner()
   Event OnParachuteClosed()
   Set random_stream(random_stream:Random Stream Structure=MakeRandomStream!(InitialSeed:int=0) -> (ReturnValue:Random Stream Structure)) -> (Output_Get:Random Stream Structure)
  ]}
]
variables [
 aimAlpha {type=number}
 armSwingDelta_l {type=number}
 armSwingDelta_r {type=number}
 crouchAlpha {type=boolean}
 headPitchAim {type=number}
 headSwingDeltaBack {type=number}
 headSwingDeltaFront {type=number}
 headTwist_r {type=number}
 MeshToCopy_0 {type=SkeletalMeshComponent}
 Not_2D_Moving {type=boolean}
 random_stream {type=RandomStream, default {initialSeed=0, seed=0}}
 rigidbody_alfa {type=number, default=1}
 skydiving vector force {type=Vector}
]
