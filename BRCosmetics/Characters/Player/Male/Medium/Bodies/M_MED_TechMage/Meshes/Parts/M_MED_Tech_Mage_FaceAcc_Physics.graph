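# Physics asset for the male medium Tech Mage face accessory mesh, setting up secondary motion for its hair and bangs.
# Capsule bodies cover each dynamic hair bone plus a large head capsule that anchors the setup.
# Strands constrain to the head directly or through small base bodies, some forming multi-segment chains such as bangs, hairA, hairN and back hair.
# Angular limits are mostly tight, except the hairN and hairP bases, which swing and twist freely.
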PhysicsAsset /BRCosmetics/Characters/Player/Male/Medium/Bodies/M_MED_TechMage/Meshes/Parts/M_MED_Tech_Mage_FaceAcc_Physics
bodies {
 dyn_bangs_01 [{shapeType=Capsule, length=1.3996, radius=0.7629, center=<1.1226 -0.2611 -0.0121>, rotation=<rot 0 -75.1141 -88.6163>}]
 dyn_bangs_02 [{shapeType=Capsule, length=1.8223, radius=0.6966, center=<1.86 -0.031 -0.0138>, rotation=<rot 0 91.309 -89.7159>}]
 dyn_hairA_01 [{shapeType=Capsule, length=0.9578, radius=1.5466, center=<1.8264 -0.3277 -0.1655>, rotation=<rot 0 18.5326 0>}]
 dyn_hairA_02 [{shapeType=Capsule, length=1.8011, radius=1.1091, center=<2.1365 0.1082 -0.3675>, rotation=<rot 0 -83.4414 -82.5455>}]
 dyn_hairB [{shapeType=Capsule, length=1.5362, radius=1.231, center=<1.2746 -0.1003 -0.0007>, rotation=<rot 0 -93.847 -178.8661>}]
 dyn_hairBackSmushA_01 [{shapeType=Capsule, length=1.38, radius=1.13, center=<1.17 0 0>, rotation=<rot 90 0 0>}]
 dyn_hairBackSmushA_01_Base [{shapeType=Capsule, length=0.505, radius=0.505, rotation=<rot 0 -90 0>}]
 dyn_hairBackSmushA_02 [{shapeType=Capsule, length=1.7327, radius=0.6851, center=<1.0877 0.0005 0.0007>, rotation=<rot 0 -87.5179 -87.3335>}]
 dyn_hairBackSmushB [{shapeType=Capsule, length=2.2176, radius=1.3435, center=<1.7713 0 0>, rotation=<rot 90 0 0>}]
 dyn_hairBackSmushB_Base [{shapeType=Capsule, length=0.505, radius=0.505, rotation=<rot 0 -90 0>}]
 dyn_hairC [{shapeType=Capsule, length=1.8222, radius=1.3398, center=<2.3601 -0.0014 0.0013>, rotation=<rot 0 94.9765 -87.5805>}]
 dyn_hairD [{shapeType=Capsule, length=2.41, radius=1.6664, center=<2.2541 -0.2183 0.2974>, rotation=<rot 0 92.499 -81.8776>}]
 dyn_hairE [{shapeType=Capsule, length=1.7735, radius=1.1595, center=<1.9094 -0.2555 0.3613>, rotation=<rot -0.5037 68.7759 -87.1421>}]
 dyn_hairF [{shapeType=Capsule, length=2.1155, radius=1.3941, center=<1.9394 -0.173 0.1704>, rotation=<rot 0 86.3077 -88.1145>}]
 dyn_hairG [{shapeType=Capsule, length=1.9368, radius=1.8145, center=<1.8659 -0.3791 0.5411>, rotation=<rot 0 149.1714 -72.7214>}]
 dyn_hairH [{shapeType=Capsule, length=1.9223, radius=1.7907, center=<0.8564 -0.0278 0.5017>, rotation=<rot 26.0707 -90 -115.967>}]
 dyn_hairJ [{shapeType=Capsule, length=2.669, radius=1.348, center=<2.657 -0.0422 0.4685>, rotation=<rot -3.0452 85.3179 -81.5955>}]
 dyn_hairK [{shapeType=Capsule, length=2.1089, radius=1.7068, center=<2.733 -0.1347 0.1617>, rotation=<rot 0 -84.4879 -88.6978>}]
 dyn_hairL [{shapeType=Capsule, length=2.6096, radius=2.1241, center=<2.2368 -0.1249 -0.0294>, rotation=<rot 0 -91.1745 -85.496>}]
 dyn_hairM [{shapeType=Capsule, length=2.3869, radius=1.6969, center=<0.949 0.1727 0.1498>, rotation=<rot 0 -116.0001 -89.1908>}]
 dyn_hairN_01 [{shapeType=Capsule, length=2.7002, radius=1.8797, rotation=<rot 90 0 0>}]
 dyn_hairN_01_Base [{shapeType=Capsule, length=0.505, radius=0.505, rotation=<rot 0 -90 0>}]
 dyn_hairN_02 [{shapeType=Capsule, length=2.2005, radius=1.3414, center=<1.6663 -0.0631 -0.0483>, rotation=<rot 0 86.2711 -89.8715>}]
 dyn_hairP [{shapeType=Capsule, length=1.7136, radius=1.134, center=<1.1696 -0.0983 0.0436>, rotation=<rot 0 82.1824 -76.3079>}]
 dyn_hairP_Base [{shapeType=Capsule, length=0.505, radius=0.505, rotation=<rot 0 -90 0>}]
 head [{shapeType=Capsule, length=7.4276, radius=4.7386, center=<5.2386 0.9702 0>, rotation=<rot 90 0 0>}]
}
constraints [
 {bone1Name=dyn_hairBackSmushB_Base, bone2Name=head, swing1LimitDegrees=15, swing2LimitDegrees=15, twistLimitDegrees=10}
 {bone1Name=dyn_hairBackSmushB, bone2Name=dyn_hairBackSmushB_Base, swing1LimitDegrees=20, swing2LimitDegrees=20, twistLimitDegrees=20}
 {bone1Name=dyn_hairBackSmushA_01_Base, bone2Name=head, swing1LimitDegrees=15, swing2LimitDegrees=15, twistLimitDegrees=10}
 {bone1Name=dyn_hairBackSmushA_01, bone2Name=dyn_hairBackSmushA_01_Base, swing1LimitDegrees=20, swing2LimitDegrees=20, twistLimitDegrees=20}
 {bone1Name=dyn_hairBackSmushA_02, bone2Name=dyn_hairBackSmushA_01, swing1LimitDegrees=20, swing2LimitDegrees=20, twistLimitDegrees=20}
 {bone1Name=dyn_hairM, bone2Name=head, swing1LimitDegrees=15, swing2LimitDegrees=15, twistLimitDegrees=10}
 {bone1Name=dyn_hairH, bone2Name=head, swing1LimitDegrees=15, swing2LimitDegrees=15, twistLimitDegrees=10}
 {bone1Name=dyn_hairJ, bone2Name=head, swing1LimitDegrees=15, swing2LimitDegrees=15, twistLimitDegrees=10}
 {bone1Name=dyn_hairD, bone2Name=head, swing1LimitDegrees=15, swing2LimitDegrees=15, twistLimitDegrees=10}
 {bone1Name=dyn_hairG, bone2Name=head, swing1LimitDegrees=15, swing2LimitDegrees=15, twistLimitDegrees=10}
 {bone1Name=dyn_hairK, bone2Name=head, swing1LimitDegrees=15, swing2LimitDegrees=15, twistLimitDegrees=10}
 {bone1Name=dyn_hairF, bone2Name=head, swing1LimitDegrees=15, swing2LimitDegrees=15, twistLimitDegrees=10}
 {bone1Name=dyn_hairC, bone2Name=head, swing1LimitDegrees=15, swing2LimitDegrees=15, twistLimitDegrees=10}
 {bone1Name=dyn_hairE, bone2Name=head, swing1LimitDegrees=15, swing2LimitDegrees=15, twistLimitDegrees=10}
 {bone1Name=dyn_hairB, bone2Name=head, swing1LimitDegrees=15, swing2LimitDegrees=15, twistLimitDegrees=10}
 {bone1Name=dyn_hairL, bone2Name=head, swing1LimitDegrees=15, swing2LimitDegrees=15, twistLimitDegrees=10}
 {bone1Name=dyn_hairP, bone2Name=dyn_hairP_Base, swing1LimitDegrees=15, swing2LimitDegrees=15, twistLimitDegrees=10}
 {bone1Name=dyn_hairN_01, bone2Name=dyn_hairN_01_Base, swing1LimitDegrees=15, swing2LimitDegrees=15, twistLimitDegrees=10}
 {bone1Name=dyn_hairN_02, bone2Name=dyn_hairN_01, swing1LimitDegrees=25, swing2LimitDegrees=25, twistLimitDegrees=10}
 {bone1Name=dyn_bangs_01, bone2Name=head, swing1LimitDegrees=15, swing2LimitDegrees=15, twistLimitDegrees=10}
 {bone1Name=dyn_bangs_02, bone2Name=dyn_bangs_01, swing1LimitDegrees=25, swing2LimitDegrees=25, twistLimitDegrees=10}
 {bone1Name=dyn_hairA_01, bone2Name=head, swing1LimitDegrees=15, swing2LimitDegrees=15, twistLimitDegrees=10}
 {bone1Name=dyn_hairA_02, bone2Name=dyn_hairA_01, swing1LimitDegrees=10, swing2LimitDegrees=10, twistLimitDegrees=10}
 {
  bone1Name=dyn_hairN_01_Base
  bone2Name=head
  swing1LimitDegrees=45
  swing1Motion=Free
  swing2LimitDegrees=45
  swing2Motion=Free
  twistLimitDegrees=45
  twistMotion=Free
 }
 {
  bone1Name=dyn_hairP_Base
  bone2Name=head
  swing1LimitDegrees=45
  swing1Motion=Free
  swing2LimitDegrees=45
  swing2Motion=Free
  twistLimitDegrees=45
  twistMotion=Free
 }
]
