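# Post-processes the Tech Mage face accessory mesh, adding physics and corrective motion to its hair bones.
# AnimGraph copies the attached parent's pose, runs RigidBody simulation relative to spine_03, then chains four additive bone adjustments weighted by backward head swing.
# Each update derives that weight from the head's ref-pose delta against spine_05, using clamped range mapping.
# Front swing, twist and pitch values are also computed but unused in AnimGraph.
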AnimBlueprint /BRCosmetics/Characters/Player/Male/Medium/Bodies/M_MED_TechMage/Meshes/Parts/M_MED_Tech_Mage_FaceAcc_AnimBP
target_skeleton=&/M_MED_Tech_Mage_FaceAcc_Skeleton
graphs [
 AnimGraph {nodes [
   n1 = RigidBody(ComponentPose=Local To Component(LocalPose=Copy Pose From Mesh() {bUseAttachedParent=true})) {
    bClampLinearTranslationLimitToRefPose=true
    bOverrideWorldGravity=true
    overridePhysicsAsset=&/M_MED_Tech_Mage_FaceAcc_Physics
    simulationSpace=BaseBoneSpace
    alphaBoolBlend {blendOption=Linear}
    baseBoneRef {boneName=spine_03}
    componentAppliedLinearAccClamp=<3000 3000 3000>
    componentLinearAccScale=<0.3 0.3 0.3>
    componentLinearVelScale=<0.4 0.4 0.4>
    overrideWorldGravity {z=-980}
   }
   n2 = Get headSwingDeltaBack() -> (headSwingDeltaBack:double)
   n3 = "Transform (Modify) Bone - Bone: dyn_hairBackSmushB_Base"(ComponentPose=@n1, Alpha=@n2, Translation="0,1.5,0", Rotation="0,0,0", Scale="1,1,1") {
    rotationMode=BMM_Additive
    rotationSpace=BCS_BoneSpace
    translationMode=BMM_Additive
    translationSpace=BCS_BoneSpace
    alphaBoolBlend {blendOption=Linear}
    alphaScaleBiasClamp {bMapRange=true, inRange {max=0.5}}
    boneToModify {boneName=dyn_hairBackSmushB_Base}
    translation {y=1.5}
   }
   n4 = "Transform (Modify) Bone - Bone: dyn_hairBackSmushA_01_Base"(ComponentPose=@n3, Alpha=@n2, Translation="0,3,0", Rotation="0,10,0", Scale="1,1,1") {
    rotationMode=BMM_Additive
    rotationSpace=BCS_BoneSpace
    translationMode=BMM_Additive
    translationSpace=BCS_BoneSpace
    alphaBoolBlend {blendOption=Linear}
    alphaScaleBiasClamp {bMapRange=true, inRange {max=0.5}}
    boneToModify {boneName=dyn_hairBackSmushA_01_Base}
    rotation {yaw=10}
    translation {y=3}
   }
   n5 = "Transform (Modify) Bone - Bone: dyn_hairN_01_Base"(ComponentPose=@n4, Alpha=@n2, Translation="0,0,0", Rotation="0,-40,0", Scale="1,1,1") {
    rotationMode=BMM_Additive
    rotationSpace=BCS_BoneSpace
    translationMode=BMM_Additive
    translationSpace=BCS_BoneSpace
    alphaBoolBlend {blendOption=Linear}
    alphaScaleBiasClamp {bMapRange=true, inRange {max=0.65}}
    boneToModify {boneName=dyn_hairN_01_Base}
    rotation {yaw=-40}
   }
   n6 = "Transform (Modify) Bone - Bone: dyn_hairP_Base"(ComponentPose=@n5, Alpha=@n2, Translation="0,2,0", Rotation="0,50,0", Scale="1,1,1") {
    rotationMode=BMM_Additive
    rotationSpace=BCS_BoneSpace
    translationMode=BMM_Additive
    translationSpace=BCS_BoneSpace
    alphaBoolBlend {blendOption=Linear}
    alphaScaleBiasClamp {bMapRange=true, inRange {max=0.5}}
    boneToModify {boneName=dyn_hairP_Base}
    rotation {yaw=50}
    translation {y=2}
   }
   Output Pose(Result=Component To Local(ComponentPose=@n6)) {layerGroup=None}
  ]}
 EventGraph {nodes [
   Event BlueprintUpdateAnimation()
   n1 = BreakTransform(InTransform=GetDeltaTransformFromRefPose(self=GetOwnerSkeletalMeshComponent(), BoneName=head, BaseName=spine_05))
   n2 = BreakRotator(InRot=@n1.Rotation)
   > Set headSwingDeltaBack(headSwingDeltaBack:double=MapRangeClamped(Value=@n2.Yaw, InRangeA=0, InRangeB=-45)) -> (Output_Get:double)
   > Set headSwingDeltaFront(headSwingDeltaFront:double=MapRangeClamped(Value=@n2.Yaw, InRangeA=0, InRangeB=25)) -> (Output_Get:double)
   n3 = BreakRotator(InRot=@n1.Rotation)
   > Set headTwist_r(headTwist_r:double=MapRangeClamped(Value=@n3, InRangeA=0, InRangeB=-61)) -> (Output_Get:double)
   > Set headPitchAim(headPitchAim:double=MapRangeClamped(Value=@n3.Pitch, InRangeA=0, InRangeB=37)) -> (Output_Get:double)
   TryGetPawnOwner()
  ]}
]
variables [
 headPitchAim {type=number}
 headSwingDeltaBack {type=number}
 headSwingDeltaFront {type=number}
 headTwist_r {type=number}
]
